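# Translucent, unlit, two-sided Niagara sprite and mesh material that draws animated electricity bolts for the FumeFleeceClap backpack effects.
# A noise-distorted, panned lightning texture, shaped by edge and optional reveal masks, forms an SDF-like field that SmoothSteps threshold into bolt and hot core.
# Particle color and an optional glow override drive emissive, while bolt plus masked glow builds opacity.
# Dynamic parameters control distortion, erosion, scroll position, glow and SplineThicken width.
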Material /BRCosmetics/Accessories/FORT_Backpacks/Backpack_M_MED_FumeFleeceClap/FX/M_FX_FumeFleeceElectricity
graph {
 properties {
  bAutomaticallySetUsageInEditor=false
  bUsedWithNiagaraMeshParticles=true
  bUsedWithNiagaraSprites=true
  blendMode=BLEND_Translucent
  shadingModel=MSM_Unlit
  twoSided=true
  usedShadingModels=Unlit
 }
 nodes [
  n1 = DynamicParameter() {defaultValue=<rgba 1 0 0 0.5>, paramNames [ThicknessMultiplier, Glow, UV.y Position, BoltShaderThickness]}
  n2 = Multiply(A=NamedRerouteUsage(), B=Multiply(A=ScalarParameter() {parameterName=Thickness}, B=@n1.ThicknessMultiplier))
  MaterialFunctionCall(WidthBase=@n2, WidthTip=@n2) {materialFunction=&SplineThicken}
  n3 = ParticleColor()
  n4 = ScalarParameter() {defaultValue=0.75, parameterName=ThinBoltCore}
  n5 = Add(A=@n4) {constB=0.1}
  n6 = NamedRerouteUsage()
  n7 = Constant3Vector() {constant=<rgba 5 5 5 0>}
  n8 = Add(A=@n3.RGB, B=Multiply(A=SmoothStep(Min=@n4, Max=@n5, Value=@n6) {constMax=0.55, constMin=0.5}, B=Multiply(A=@n3.RGB, B=@n7)))
  n9 = VectorParameter() {parameterName=GlowColorOverride}
  n10 = Multiply(A=Power(Base=@n6)) {constB=0.125}
  n11 = TextureCoordinate()
  n12 = ComponentMask(None=@n11) {g=true}
  n13 = DynamicParameter() {defaultValue=<rgba 1 0 0 0.5>, paramNames [ThicknessMultiplier, Glow, UV.y Position, BoltShaderThickness]}
  n14 = Add(A=@n13.BoltShaderThickness, B=ScalarParameter() {defaultValue=0.01, group=Shape Control, parameterName=BoltSharpness}) {constB=0.2}
  n15 = Multiply(
   A=Add(A=Multiply(A=@n10, B=@n12), B=SmoothStep(Min=@n13.BoltShaderThickness, Max=@n14, Value=@n6) {constMax=0.55, constMin=0.5})
   B=SphereMask(A=ComponentMask(None=@n11) {r=true}, B=Constant() {r=0.5}) {attenuationRadius=0.5, hardnessPercent=0}
  )
  n16 = Saturate(None=@n15)
  n17 = DynamicParameter() {
   parameterIndex=2
   defaultValue=<rgba 0 0 0.5 0.5>
   paramNames [
    ThicknessMultiplier/NoiseDistortion/UseGlowColorOverride
    UV.x Index/Speed Y
    UV.y Position/NoiseErosion/MaskWidth
    BoltShaderThickness/NoiseDistortion/MaskPosition
   ]
  }
  MaterialFunctionCall(Emissive=LinearInterpolate(A=@n8, B=LinearInterpolate(A=@n9.RGB, B=@n8, Alpha=@n16), Alpha=@n17.ThicknessMultiplier/NoiseDistortion/UseGlowColorOverride)) {materialFunction=&MF_FX_EmissiveAdjustments}
  n18 = ParticleColor()
  n19 = Constant() {r=0.5}
  n20 = SphereMask(A=TextureCoordinate(), B=@n19) {attenuationRadius=0.5, hardnessPercent=0}
  n21 = TextureCoordinate()
  n22 = Multiply(A=NamedRerouteUsage(), B=SphereMask(A=ComponentMask(None=@n21) {r=true}, B=@n19) {attenuationRadius=0.5, hardnessPercent=0})
  n23 = ScalarParameter() {group=AnimatedReveal, parameterName=UseAnimatedRevealMask}
  n24 = Multiply(
   A=LinearInterpolate(B=@n9.A, Alpha=@n17.ThicknessMultiplier/NoiseDistortion/UseGlowColorOverride) {constA=1}
   B=Power(Base=LinearInterpolate(A=@n20, B=@n22, Alpha=@n23)) {constExponent=4}
  )
  n25 = DynamicParameter() {defaultValue=<rgba 1 0 0 0.5>, paramNames [ThicknessMultiplier, Glow, UV.y Position, BoltShaderThickness]}
  Multiply(
   A=Add(A=Multiply(A=@n15, B=@n18.A), B=Multiply(A=@n24, B=@n25.Glow)) {constB=0.1}
   B=LinearInterpolate(B=VertexColor().R, Alpha=ScalarParameter() {parameterName=VertexColorMask}) {constA=1}
  )
  n26 = ComponentMask(None=TextureCoordinate()) {g=true}
  n27 = Saturate(None=Multiply(A=OneMinus(None=@n26)) {constB=16})
  n28 = TextureCoordinate()
  n29 = DynamicParameter() {
   parameterIndex=1
   defaultValue=<rgba 0 0 0 0.5>
   paramNames [ThicknessMultiplier/NoiseDistortion, UV.x Index/Speed Y, UV.y Position/NoiseErosion, BoltShaderThickness/NoiseDistortion]
  }
  n30 = Add(A=OneMinus(None=ComponentMask(None=@n28) {g=true}), B=@n29.UV.y Position/NoiseErosion)
  n31 = VectorParameter() {parameterName=NoiseScale, defaultValue=<rgba 0 0 0 1>}
  n32 = Multiply(A=TextureCoordinate(), B=Multiply(A=ComponentMask(None=@n31.RGB) {g=true, r=true}, B=@n31.B))
  n33 = Multiply(A=TextureSample(UVs=Panner(Coordinate=@n32) {bFractionalPart=true, speedY=1}) {texture=/BRCosmetics/Athena/Effects/Particles/Skydiving/NS_Hightower_Date/T_AtmosphericCloudNoise_Vertical}.R) {constB=2}
  n34 = TextureCoordinate()
  n35 = ComponentMask(None=VectorParameter() {group=UV, parameterName=Tiling, defaultValue=<rgba 6 0.75 0 1>}.RGB) {g=true, r=true}
  n36 = DynamicParameter() {defaultValue=<rgba 1 0 0 0.5>, paramNames [ThicknessMultiplier, Glow, UV.y Position, BoltShaderThickness]}
  n37 = Add(
   A=Multiply(A=@n29.ThicknessMultiplier/NoiseDistortion, B=Subtract(A=@n33))
   B=Panner(
    Coordinate=MaterialFunctionCall(UVs=@n34, Texture Scale=@n35) {materialFunction=&ScaleUVsByCenter}.UVs
    Time=@n36.UV.y Position
   ) {bFractionalPart=true, speedY=1.1}
  )
  n38 = TextureCoordinate() {vTiling=2}
  n39 = Multiply(
   A=Multiply(
    A=Saturate(None=@n30)
    B=TextureSample(UVs=@n37) {texture=/Game/Effects/Fort_Effects/GameplayPulginsAssets/ToonLightning/MasterMaterial/V2/LightningTexture}.G
   )
   B=Saturate(None=ComponentMask(None=@n38) {g=true})
  )
  NamedRerouteDeclaration(None=Multiply(A=Multiply(A=@n27, B=@n39), B=NamedRerouteUsage())) {name=LightningSDF, nodeColor=<rgba 0.8047 1 0 1>}
  NamedRerouteDeclaration(None=@n27) {name=LeadingEdgeMask, nodeColor=<rgba 1 0 0.8438 1>}
  n40 = ComponentMask(None=TextureCoordinate()) {g=true}
  n41 = DynamicParameter() {
   parameterIndex=2
   defaultValue=<rgba 0 0 0.5 0.5>
   paramNames [
    ThicknessMultiplier/NoiseDistortion/UseGlowColorOverride
    UV.x Index/Speed Y
    UV.y Position/NoiseErosion/MaskWidth
    BoltShaderThickness/NoiseDistortion/MaskPosition
   ]
  }
  NamedRerouteDeclaration(None=LinearInterpolate(
    B=SphereMask(A=@n40, B=@n41.BoltShaderThickness/NoiseDistortion/MaskPosition, Radius=@n41.BoltShaderThickness/NoiseDistortion/MaskPosition) {attenuationRadius=0.5, hardnessPercent=0}
    Alpha=ScalarParameter() {group=AnimatedReveal, parameterName=UseAnimatedRevealMask}
   ) {constA=1}) {name=MaskReveal, nodeColor=<rgba 0 0.9062 1 1>}
 ]
}
