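# Translucent, unlit, two-sided Niagara sprite and mesh-particle material for light shaft effects on the FumeFleeceClap backpack cosmetic.
# Particle color feeds an emissive adjustment function.
# A U-axis gradient multiplied by panning packed noise, timed by a dynamic parameter, passes through SmoothStep and particle alpha to form a mask.
# A vertex-offset-style vector points away from the pivot, scaled by powered U and the dynamic Expansion parameter.
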Material /BRCosmetics/Accessories/FORT_Backpacks/Backpack_M_MED_FumeFleeceClap/FX/M_FX_FumeFleeceLightShafts
graph {
 properties {
  bAutomaticallySetUsageInEditor=false
  bUsedWithNiagaraMeshParticles=true
  bUsedWithNiagaraSprites=true
  blendMode=BLEND_Translucent
  shadingModel=MSM_Unlit
  twoSided=true
  usedShadingModels=Unlit
 }
 nodes [
  n1 = ParticleColor()
  MaterialFunctionCall(Emissive=@n1.RGB) {materialFunction=&MF_FX_EmissiveAdjustments}
  n2 = ScalarParameter() {parameterName=A}
  n3 = Multiply(A=ComponentMask(None=TextureCoordinate()) {r=true}, B=ScalarParameter() {defaultValue=1, parameterName=GradientScale})
  n4 = VectorParameter() {parameterName=NoiseScale, defaultValue=<rgba 0 0 0 1>}
  n5 = Panner(
   Coordinate=Multiply(A=TextureCoordinate(), B=ComponentMask(None=@n4.RGB) {g=true, r=true})
   Time=DynamicParameter() {defaultValue=<rgba 10 1 1 1>, paramNames [Expansion,NoiseTime,Param3,Param4]}.NoiseTime
   Speed=ComponentMask(None=VectorParameter() {parameterName=NoiseSpeed, defaultValue=<rgba 0 0 0 1>}.RGB) {g=true, r=true}
  ) {bFractionalPart=true, speedY=1}
  n6 = Subtract(
   A=ChannelMaskParameter(None=TextureSample(UVs=@n5) {texture=/Game/Effects/Fort_Effects/Textures/Fire/HF_FireNoisesPacked}.RGB) {parameterName=NoiseChannel}.R
   B=ScalarParameter() {parameterName=NoiseOffset}
  )
  n7 = ScalarParameter() {parameterName=NoiseMultiplier}
  n8 = SmoothStep(Min=@n2, Max=Add(A=@n2, B=ScalarParameter() {parameterName=B}), Value=Multiply(A=Clamp(None=@n3), B=Multiply(A=@n6, B=@n7)))
  Multiply(A=@n8, B=@n1.A)
  n9 = TextureCoordinate()
  n10 = Subtract(
   A=WorldPosition().XYZ
   B=MaterialFunctionCall() {materialFunction=&ObjectPivotPoint}.Mesh Particle Pivot Location
  )
  Multiply(
   A=Power(Base=ComponentMask(None=@n9) {r=true}, Exp=ScalarParameter() {defaultValue=2, parameterName=Shape})
   B=Multiply(A=Normalize(VectorInput=@n10), B=DynamicParameter() {defaultValue=<rgba 10 1 1 1>, paramNames [Expansion,NoiseTime,Param3,Param4]}.Expansion)
  )
  OneMinus(None=@n8)
  SmoothStep(
   Min=ScalarParameter() {parameterName=EdgeMask_Min}
   Max=ScalarParameter() {parameterName=EdgeMask_Max}
   Value=ComponentMask(None=TextureCoordinate()) {r=true}
  )
  Transform(None=VectorParameter() {parameterName=ExpansionDirection, defaultValue=<rgba 0 1 0 1>}.RGB)
 ]
}
