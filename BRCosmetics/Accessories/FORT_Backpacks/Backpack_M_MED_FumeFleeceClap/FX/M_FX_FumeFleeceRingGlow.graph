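# Translucent, unlit, two-sided Niagara material for a ring glow effect on the FumeFleeceClap backpack.
# A SplineThicken function takes its width from Thickness and a dynamic multiplier, while particle color feeds an emissive-adjustment function.
# Opacity is a soft U-centered SphereMask raised to a dynamic exponent, scaled by Glow and particle alpha, then depth-faded.
# Panning cloud noise drives the exponent lerp, but both ends use GlowExponentMin, so it has no effect.
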Material /BRCosmetics/Accessories/FORT_Backpacks/Backpack_M_MED_FumeFleeceClap/FX/M_FX_FumeFleeceRingGlow
graph {
 properties {
  bAutomaticallySetUsageInEditor=false
  bUsedWithNiagaraMeshParticles=true
  bUsedWithNiagaraSprites=true
  blendMode=BLEND_Translucent
  shadingModel=MSM_Unlit
  twoSided=true
  usedShadingModels=Unlit
 }
 nodes [
  n1 = TextureCoordinate()
  n2 = DynamicParameter() {defaultValue=<rgba 1 0 2 10>, paramNames [ThicknessMultiplier, Glow, GlowExponentMin, GlowExponentMax]}
  n3 = Panner(Coordinate=TextureCoordinate() {uTiling=0.2}) {speedX=0.5}
  n4 = Power(
   Base=SphereMask(A=ComponentMask(None=@n1) {r=true}, B=Constant() {r=0.5}) {attenuationRadius=0.5, hardnessPercent=0}
   Exp=LinearInterpolate(
    A=@n2.GlowExponentMin
    B=@n2.GlowExponentMin
    Alpha=TextureSample(UVs=@n3) {texture=/Game/Effects/Fort_Effects/Textures/Smoke/T_Clouds_02_M}.G
   )
  ) {constExponent=4}
  n5 = DynamicParameter() {defaultValue=<rgba 1 0 2 10>, paramNames [ThicknessMultiplier, Glow, GlowExponentMin, GlowExponentMax]}
  n6 = ParticleColor()
  DepthFade(Opacity=Multiply(A=Multiply(A=@n4, B=@n5.Glow), B=@n6.A), FadeDistance=ScalarParameter() {defaultValue=5, parameterName=DepthFade})
  MaterialFunctionCall(Emissive=@n6.RGB) {materialFunction=&MF_FX_EmissiveAdjustments}
  n7 = Multiply(A=ScalarParameter() {parameterName=Thickness}, B=@n2.ThicknessMultiplier)
  MaterialFunctionCall(WidthBase=@n7, WidthTip=@n7) {materialFunction=&SplineThicken}
  ScalarParameter() {defaultValue=2, parameterName=GlowExponent_Min}
  ScalarParameter() {defaultValue=4, parameterName=GlowExponent_Max}
 ]
}
