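# Translucent unlit, two-sided Niagara overlay material for the FumeFleeceClap backpack effect, showing a wipe or solid flash frames.
# A scaled, rotated gradient texture panned by dynamic WipePan masks particle color.
# A dynamic State value lerps between that wipe, WhiteFrame and BlackFrame colors, though its label lists black before white.
# RGB feeds emissive adjustments, alpha is multiplied by vertex red, and CameraOffset shifts position.
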Material /BRCosmetics/Accessories/FORT_Backpacks/Backpack_M_MED_FumeFleeceClap/FX/M_FX_FumeFleeceWipeOverlay
graph {
 properties {
  bAutomaticallySetUsageInEditor=false
  bUsedWithNiagaraMeshParticles=true
  bUsedWithNiagaraSprites=true
  blendMode=BLEND_Translucent
  shadingModel=MSM_Unlit
  twoSided=true
  usedShadingModels=Unlit
 }
 nodes [
  MaterialFunctionCall(Offset Amount=ScalarParameter() {parameterName=CameraOffset}) {materialFunction=/Engine/Functions/Engine_MaterialFunctions02/WorldPositionOffset/CameraOffset}
  n1 = ParticleColor()
  n2 = Multiply(A=TextureCoordinate(), B=ScalarParameter() {defaultValue=1, parameterName=WipeScale})
  n3 = ScalarParameter() {parameterName=WipeRotation}
  n4 = DynamicParameter() {defaultValue=<rgba 0 0 1 1>, paramNames [WipePan, "State (0 - Wipe, 1 - BlackFrame , 2 - WhiteFrame)", Param3, Param4]}
  n5 = TextureSample(UVs=Add(A=Rotator(Coordinate=@n2, Time=@n3), B=@n4.WipePan)) {
   samplerSource=SSM_Clamp_WorldGroupSettings
   samplerType=SAMPLERTYPE_LinearGrayscale
   texture=/Game/Effects/Fort_Effects/Textures/Grad/T_Vertical_Gradient_Wave
  }
  n6 = VectorParameter() {parameterName=BlackFrame}
  n7 = MaterialFunctionCall!(
   1=AppendVector(A=@n1.RGB, B=@n5.R)
   2=AppendVector(A=VectorParameter() {parameterName=WhiteFrame}.RGB, B=@n1.A)
   3=AppendVector(A=@n6.RGB, B=@n6.RGB)
   4
   A=Divide(A=@n4."State (0 - Wipe, 1 - BlackFrame , 2 - WhiteFrame)")
  ) {materialFunction=/Engine/Functions/Engine_MaterialFunctions02/Lerp_Multiple_Float4}
  MaterialFunctionCall(Emissive=ComponentMask(None=@n7.Lerp 3 Inputs) {b=true, g=true, r=true}) {materialFunction=&MF_FX_EmissiveAdjustments}
  Multiply(A=ComponentMask(None=@n7.Lerp 3 Inputs) {a=true}, B=VertexColor().R)
 ]
}
